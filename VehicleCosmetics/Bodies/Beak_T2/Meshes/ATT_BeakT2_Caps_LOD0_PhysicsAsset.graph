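# Physics asset for the LOD0 mesh of the Beak T2 caps attachment, a vehicle body cosmetic.
# It approximates the attachment with a single body holding one large, rotated capsule, offset forward and upward from the origin.
# It has no constraints, so it defines no joints between bodies.
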PhysicsAsset /VehicleCosmetics/Bodies/Beak_T2/Meshes/ATT_BeakT2_Caps_LOD0_PhysicsAsset
bodies {body [{shapeType=Capsule, length=163.6559, radius=95.8742, center=<34.7582 0 45.6475>, rotation=<rot 0 -90 -86.1882>}]}
constraints []
