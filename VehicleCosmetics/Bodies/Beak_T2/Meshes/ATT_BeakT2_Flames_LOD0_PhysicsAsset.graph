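# Physics asset for the Beak T2 flames attachment skeletal mesh (LOD0), part of a vehicle body cosmetic.
# It holds a single physics body named "body", shaped as one rotated capsule offset slightly from the bone origin.
# It defines no constraints, so the asset provides one simple collision volume rather than a jointed ragdoll.
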PhysicsAsset /VehicleCosmetics/Bodies/Beak_T2/Meshes/ATT_BeakT2_Flames_LOD0_PhysicsAsset
bodies {body [{shapeType=Capsule, length=143.5838, radius=79.4024, center=<-15.7913 -0.0003 5.8868>, rotation=<rot 0 -89.9998 -71.0317>}]}
constraints []
